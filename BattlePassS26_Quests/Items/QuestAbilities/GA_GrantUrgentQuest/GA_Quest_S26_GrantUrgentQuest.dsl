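; Gameplay ability for a Season 26 Battle Pass quest that grants an urgent quest to the owning player.
; On activation it starts two gameplay tag-add wait tasks on Visit tags, whose Added outputs do nothing.
; It looks up the configured quest definition on the player state and grants it through the transient quests component only if absent, preventing double grants.
; The end-ability event and CheckForNearbyEnemies function are empty.
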(event Ability|EventActivateAbility
  ;[1 1 1] Check if player already has the urgent quest, then grant if they dont (don't double grant quest)
  (|CallParentFunction)
  (Ability|Tasks|WaitGameplayTagAdd (Variables|Default|GetVisit)
    (:then
      (Ability|Tasks|WaitGameplayTagAdd (Variables|Default|GetVisit2)
        (:Added
          _)))
    (:Added
      _))
  (bind _returnvalue (Quest|GetQuestfromDefinition (Variables|/Script/Engine.Controller|GetPlayerState (Fortnite|GetFortPlayerControllerfromActor (Variables|GAVars|Data|GetOwningPlayerPawn))) (Variables|Default|GetQuest%20Def%20Athena%20to%20Grant)))
  (Utilities|IsValid _returnvalue
    (:Is%20Not%20Valid
      (Class|FortControllerComponentTransientQuests|GrantTransientQuest (Class|FortPlayerPawn|GetTransientQuestsComponent (Variables|GAVars|Data|GetOwningPlayerPawn)) (Variables|Default|GetQuest%20Def%20Athena%20to%20Grant) 0))))

(event Ability|EventOnEndAbility
  (params (bWasCancelled :type Boolean)))

(fn CheckForNearbyEnemies
  (params (LastCheck1 :type Boolean)))
