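; Passive Season 26 quest ability that grants an urgent quest when the player loots high-rarity items inside a heist base.
; Two parallel gameplay-event waits watch for locking display case interactions, used as a proxy for high-rarity loot, and pickups with required rarity tags.
; A match leads to a check of the pawn's location tags.
; If the player lacks the quest, it is granted as a transient quest after a short delay.
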(event Ability|EventActivateAbility
  ;[1 1 1] Activate Passive Ability
  (|CallParentFunction)
  (CallFunction|CheckPlayerTags))

(event Custom|CheckPlayerTags
  ;[1 1 1] Check if player has looted a SR+ rarity item and if player is in a heist base
  ;;[0.3802 0 0] We are making an assumption here that any item in a Display Case qualifies. Ideally we would have a quest hook for this in the future - Katy
  (bind (_asynctaskproxy _payload) (Ability|Tasks|WaitGameplayEvent "(TagName=\"Event.Interact.Completed\")")
    (:then
      (bind (_asynctaskproxy_2 _payload_2) (Ability|Tasks|WaitGameplayEvent "(TagName=\"Event.CollectedPickup\")")
        (:EventReceived
          (bind _returnvalue (GameplayTags|GetOwnedGameplayTags (.Target (Utilities|Struct|BreakGameplayEventData _payload_2)) :!class /Script/GameplayTags.GameplayTagAssetInterface))
          (if (GameplayTags|HasAnyTags _returnvalue (Variables|Default|GetRequiredRarityTags))
            _))))
    (:EventReceived
      (if (GameplayTags|HasTag (.TargetTags (Utilities|Struct|BreakGameplayEventData _payload)) "(TagName=\"Athena.Quests.LockingDisplayCase\")" true)
        _)))
  (if (GameplayTags|HasAnyTags (GameplayTags|GetOwnedGameplayTags (Variables|GAVars|Data|GetOwningPlayerPawn) :!class /Script/GameplayTags.GameplayTagAssetInterface) (Variables|Default|GetRequiredLocationTags))
    (CallFunction|StartUrgentQuest)))

(event Custom|StartUrgentQuest
  ;[1 1 1] Grant the associated urgent quest for this action
  (bind _returnvalue (Quest|GetQuestfromDefinition (Variables|/Script/Engine.Controller|GetPlayerState (Fortnite|GetFortPlayerControllerfromActor (Variables|GAVars|Data|GetOwningPlayerPawn))) (Variables|Default|GetQuest%20Def%20Athena%20to%20Grant)))
  (Utilities|IsValid _returnvalue
    (:Is%20Not%20Valid
      (Utilities|FlowControl|Delay 3.0)
      (Class|FortControllerComponentTransientQuests|GrantTransientQuest (Class|FortPlayerPawn|GetTransientQuestsComponent (Variables|GAVars|Data|GetOwningPlayerPawn)) (Variables|Default|GetQuest%20Def%20Athena%20to%20Grant) 0))))

(event Ability|EventOnEndAbility
  (params (bWasCancelled :type Boolean))
  (|CallParentFunction bWasCancelled))

(event Ability|EventFailedToActivatePassiveAbility
  (params (ActorInfo :type Gameplay%20Ability%20Actor%20Info%20Structure))
  (|CallParentFunction ActorInfo))

(fn CheckForNearbyEnemies
  (params (LastCheck1 :type Boolean)))
