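; Season 26 quest gameplay ability that marks the owning player with the UrgentQuestActive gameplay effect while the ability runs.
; On activation it calls the parent logic, applies the effect to the owning pawn's ability system component and stores the active handle.
; When the ability ends, it removes that effect by the stored handle.
; The CheckForNearbyEnemies function has an empty body here.
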(event Ability|EventActivateAbility
  (|CallParentFunction)
  (bind _returnvalue (GameplayEffects|ApplyGameplayEffectToSelf :self (Ability|GetFortAbilitySystemComponent (Variables|GAVars|Data|GetOwningPlayerPawn)) :GameplayEffectClass "/BattlePassS26_Quests/Items/QuestAbilities/GA_GrantUrgentQuest/GE_Quest_UrgentQuestActive.GE_Quest_UrgentQuestActive_C" :Level 0.0))
  (Variables|Default|SetActiveGE _returnvalue))

(event Ability|EventOnEndAbility
  (params (bWasCancelled :type Boolean))
  (GameplayEffects|RemoveActiveGameplayEffect (Ability|GetFortAbilitySystemComponent (Variables|GAVars|Data|GetOwningPlayerPawn)) (Variables|Default|GetActiveGE) -1))

(fn CheckForNearbyEnemies
  (params (LastCheck1 :type Boolean)))
